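# Additive unlit Niagara mesh-particle material for the War Effort donate box, building a light band that wraps onto scene geometry.
# World position behind translucency, relative to the pivot, drives a vertical gradient lookup shifted by dynamic BandPosition.
# An inverted SmoothStep masks it to scene points near the mesh surface, scaled by particle alpha.
# A camera-offset function is present; material output links are not shown.
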Material /WarEffort/FX/Materials/M_WarEffort_DonateBox_Lightwrapper
graph {
 properties {
  bUsedWithNiagaraMeshParticles=true
  blendMode=BLEND_Additive
  pixelDepthOffsetMode=PDOM_Legacy
  shadingModel=MSM_Unlit
  usedShadingModels=Unlit
 }
 nodes [
  Constant3Vector() {constant=<rgba 1 0 0 0>}
  MaterialFunctionCall() {materialFunction=&LocalPosition}
  MaterialFunctionCall(Offset Amount=ScalarParameter() {parameterName=Vertex_CamOffset}) {materialFunction=/Engine/Functions/Engine_MaterialFunctions02/WorldPositionOffset/CameraOffset}
  n1 = Subtract(
   A=MaterialFunctionCall() {materialFunction=&WorldPositionBehindTranslucency}.Result
   B=MaterialFunctionCall() {materialFunction=&ObjectPivotPoint}.Object Pivot Location
  )
  n2 = Divide(
   A=MaterialFunctionCall(Float3=@n1) {materialFunction=&BreakOutFloat3Components}.B
   B=ScalarParameter() {parameterName=GradSize}
  ) {constB=200}
  n3 = TextureSample(UVs=Add(A=OneMinus(None=@n2), B=DynamicParameter() {paramNames [BandPosition,Param2,Param3,Param4]}.BandPosition)) {texture=/Game/Effects/Fort_Effects/Textures/Grad/gradient_healed}
  n4 = ScalarParameter() {parameterName=MeshWrapTolerance}
  n5 = ScalarParameter() {parameterName=WrapTolerance_Smoothout}
  n6 = Add(A=@n4, B=@n5)
  n7 = Subtract(A=@n4, B=@n5)
  n8 = WorldPosition()
  n9 = ParticlePositionWS()
  n10 = MaterialFunctionCall(Vector 3=Subtract(A=@n1, B=Subtract(A=@n8.XYZ, B=@n9))) {materialFunction=/Engine/Functions/Engine_MaterialFunctions02/Utility/VectorLength}
  Multiply(A=ParticleColor().A, B=Multiply(A=Multiply(A=@n3.G) {constB=5}, B=SmoothStep(Min=@n6, Max=@n7, Value=@n10.V3 Length)))
  ParticlePositionWS()
 ]
}
