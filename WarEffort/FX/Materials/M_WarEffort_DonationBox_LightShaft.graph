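# Unlit additive material for Niagara mesh particles that draws the War Effort donation box light shaft.
# A radial UV mask blends particle color toward a desaturated version and between min and max brightness, scaled by dynamic OuterBrightness.
# A separate mask multiplies particle alpha, a flare-streak texture scaled about its center by dynamic UVScale, gradients from both UV channels, vertex red and an inverted Fresnel that fades grazing angles.
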Material /WarEffort/FX/Materials/M_WarEffort_DonationBox_LightShaft
graph {
 properties {
  bUsedWithNiagaraMeshParticles=true
  blendMode=BLEND_Additive
  pixelDepthOffsetMode=PDOM_Legacy
  shadingModel=MSM_Unlit
  usedShadingModels=Unlit
 }
 nodes [
  ComponentMask(None=TextureCoordinate() {vTiling=0.5}) {g=true}
  n1 = SphereMask(A=TextureCoordinate(), B=Constant() {r=0.5}) {attenuationRadius=0.5, hardnessPercent=0}
  n2 = ParticleColor()
  n3 = Desaturation(None=@n2.RGB, Fraction=ScalarParameter() {parameterName=max_desat})
  n4 = DynamicParameter() {defaultValue=<rgba 10 1 1 1>, paramNames [UVScale, OuterBrightness, Param3, Param4]}
  Multiply(
   A=LinearInterpolate(A=ScalarParameter() {parameterName=min_brightness}, B=ScalarParameter() {parameterName=max_brightness}, Alpha=@n1)
   B=Multiply(A=LinearInterpolate(A=@n2.RGB, B=@n3, Alpha=@n1), B=@n4.OuterBrightness) {constB=12}
  )
  n5 = OneMinus(None=Fresnel(ExponentIn=ScalarParameter() {parameterName=angle_fade}) {baseReflectFraction=0})
  n6 = MaterialFunctionCall(Texture Scale=@n4.UVScale) {materialFunction=&ScaleUVsByCenter}
  n7 = TextureCoordinate() {coordinateIndex=1}
  n8 = Multiply(
   A=Multiply(A=@n2.A, B=TextureSample(UVs=@n6.UVs) {texture=/Game/Effects/Fort_Effects/Textures/Flares/T_Flare_Streaks}.G)
   B=OneMinus(None=ComponentMask(None=@n7) {g=true})
  )
  Multiply(A=ComponentMask(None=TextureCoordinate()) {g=true}, B=Multiply(A=Multiply(A=@n5, B=@n8), B=VertexColor().R))
  ScalarParameter() {defaultValue=6.487, parameterName=Scale}
 ]
}
